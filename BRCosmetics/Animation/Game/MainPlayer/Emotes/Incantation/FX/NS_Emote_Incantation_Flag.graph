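# Niagara system for the Incantation emote's flag effect: a brief, very bright red-orange flash.
# A single FlagFLash emitter fires an instantaneous burst of short-lived particles, drawn through one sprite and two mesh renderers.
# ScaleColor drives an HDR color curve and an alpha fade, with alpha strength scaled by time of day and a gameplay versus front-end modifier.
# The sprite alignment and mesh-size animation modules are disabled.
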NiagaraSystem /BRCosmetics/Animation/Game/MainPlayer/Emotes/Incantation/FX/NS_Emote_Incantation_Flag
dependencies {
 usedDataInterfaces [&NiagaraDataInterfaceCurve, &NiagaraDataInterfaceColorCurve]
 usedDynamicInputs []
 usedModules [
  &SystemState
  &self:SystemSpawnScript.SystemScriptSource.SystemScriptGraph.NiagaraNodeAssignment_0.SetVariables_#1
 ]
}
emitters {FlagFLash {
  emitterUpdateScript [
   &EmitterState()
   &SpawnBurst_Instantaneous(Use Spawn Probability=false, UseLoopCountLimit=false)
  ]
  particleSpawnScript [
   &InitializeParticle(
    Lifetime=0.15
    Mesh Scale=<1 10 1>
    Mesh Scale Mode=::NewEnumerator4
    Position Offset Coordinate Space=::Local
    Sprite Size Mode=::NewEnumerator3
    Uniform Sprite Size=68
    UsePositionOffset=true
    Write Material Random=true
   )
   &CameraOffset()
   &SpriteFacingAndAlignment(
    Alignment Coordinate Space=::Local
    Facing Coordinate Space=::Local
    Sprite Alignment=<0 0 0>
    Sprite Facing=<0 1 0>
    WriteAlignment=true
    WriteFacing=true
   ) {enabled=false}
  ]
  particleUpdateScript [
   &ParticleState()
   &ScaleMeshSize(Scale Factor=&VectorFromFloat(Value=&FloatFromCurve(FloatCurve=DataInterface() {propertyValues {Curve {keys [<at 0 1.2>,<at 1 1>]}}}))) {enabled=false}
   &ScaleColor(
    Scale Alpha=&FloatFromCurve(
     FloatCurve=DataInterface() {propertyValues {Curve {keys [
         {arriveTangent=-2.6629, interpMode=RCIM_Cubic, leaveTangent=-2.6629, tangentMode=RCTM_User, time=0, value=1}
         {arriveTangent=-0.0145, tangentMode=RCTM_User, time=1, value=0}
        ]}}}
     Scale Curve=&Multiply_Float(
      A=&FNTimeOfDayScalar(Night Scale=0.5, Morning Scale=0.6, Evening Scale=0.6)
      B=&FN_FrontEndFloatModifier(Gameplay=40, Front End=0.2)
     )
    )
    Scale RGB=&MakeVectorFromLinearColorRGB(Color=&ColorFromCurve(ColorCurve=DataInterface() {propertyValues {
        AlphaCurve {keys [<at 0.1221 1 RCIM_Cubic>,<at 1 0 RCIM_Cubic>]}
        BlueCurve {keys [<at 0.1654 1.6 RCIM_Cubic>,<at 0.2627 0>]}
        GreenCurve {keys [<at 0.1654 17.28 RCIM_Cubic>,<at 0.2627 0.3472>]}
        RedCurve {keys [<at 0.1654 80 RCIM_Cubic>,<at 0.2627 40>]}
       }}))
    ScaleA=true
    ScaleRGB=true
   )
  ]
  rendererClasses [&NiagaraSpriteRendererProperties, &NiagaraMeshRendererProperties]
  renderers [
   {rendererClass=&NiagaraSpriteRendererProperties}
   {rendererClass=&NiagaraMeshRendererProperties, rendererIndex=1}
   {rendererClass=&NiagaraMeshRendererProperties, rendererIndex=2}
  ]
 }}
summary {
 systemName=NS_Emote_Incantation_Flag
 emitters [{emitterName=FlagFLash, rendererClasses [&NiagaraSpriteRendererProperties, &NiagaraMeshRendererProperties]}]
 userVariables [
  User.bIsFrontEnd {type=NiagaraBool}
  User.bIsFrontEndPreview {type=NiagaraBool}
 ]
}
